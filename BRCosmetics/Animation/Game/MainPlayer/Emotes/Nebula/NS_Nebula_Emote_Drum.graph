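# Niagara drum-hit effect for the Nebula emote, scaled by User.HitIntensity.
# Embers bursts and streams purple-blue sprites from a flattened, rotated sphere shape, driven by velocity, drag and curl noise.
# Empty bursts two short-lived particles rendered as sprite, mesh and light, with curve-animated scale, color and material parameters.
# It stores intensity at spawn for sprite alpha and light brightness; User.Hit feeds only a disabled SpawnRate.
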NiagaraSystem /BRCosmetics/Animation/Game/MainPlayer/Emotes/Nebula/NS_Nebula_Emote_Drum
dependencies {
 usedDataInterfaces [&NiagaraDataInterfaceCurve, &NiagaraDataInterfaceColorCurve]
 usedDynamicInputs []
 usedModules [&SystemState]
}
emitters {
 Embers {
  emitterUpdateScript [
   &EmitterState()
   &SpawnBurst_Instantaneous(
    Spawn Count=&MakeIntFromFloat(Float=&Multiply_Float(A=200, B=&Add_Float(A=$User.HitIntensity, B=0.1)))
    Use Spawn Probability=false
    UseLoopCountLimit=false
   )
   &SpawnRate(
    SpawnRate=&Multiply_Float(A=500, B=&Add_Float(A=$User.HitIntensity, B=0.3))
    Use Spawn Probability=false
   )
  ]
  particleSpawnScript [
   &InitializeParticle(
    Color Maximum=<rgba 0.1106 0 1 1>
    Color Minimum=<rgba 0.919 0 2.4174 1>
    Color Mode=::NewEnumerator2
    Lifetime Max=0.6
    Lifetime Min=0.2
    Lifetime Mode=::NewEnumerator1
    Sprite Size Mode=::NewEnumerator3
    Uniform Sprite Size=&Multiply_Float(
     A=&RandomRangeFloat(Minimum=1.5, Maximum=4)
     B=&Add_Float(A=&Multiply_Float(A=$User.HitIntensity, B=0.5), B=0.5)
    )
    UsePositionOffset=false
    Write Material Random=true
   )
   &ShapeLocation(
    Hemisphere Distribution=<0.5 1>
    MaskSpawnGroup=false
    Non Uniform Scale=<1 0.3 1>
    Offset=<0 37 0>
    Offset Mode=::NewEnumerator0
    Override Seed=false
    Sphere Radius=&Multiply_Float(A=55, B=&Add_Float(A=0.7, B=&Multiply_Float(A=0.3, B=$User.HitIntensity)))
    Sphere Surface Distribution=1
    UseOwnerScale=true
   )
   /SetVariables(Particles.Position=&RotateVectorByQuaternion(
     Vector To Rotate=Linked() {linkedVariable {name=Particles.Position, type=NiagaraPosition}}
     Quaternion=&MakeQuatFromAxisAngle(Axis=<0 0 1>, AngleInDegrees=-90)
    )) {isSetParameters=true}
   &AddVelocity(
    Module.Use Falloff Distance=false
    Origin Offset=<40 0 0>
    Override Seed=false
    Velocity Mode=::NewEnumerator1
    Velocity Speed=&Multiply_Float(
     A=&RandomRangeFloat(Minimum=150, Maximum=300)
     B=&Add_Float(A=$User.HitIntensity, B=0.3)
    )
   )
   &ApplyInitialForces()
  ]
  particleUpdateScript [
   &ParticleState()
   &ScaleSpriteSize(Uniform Curve Sprite Scale=DataInterface() {propertyValues {Curve {keys [
        {arriveTangent=13.2638, interpMode=RCIM_Cubic, leaveTangent=13.2638, tangentMode=RCTM_User, time=0, value=0}
        <at 0.1 1>
        {arriveTangent=-1.538, interpMode=RCIM_Cubic, leaveTangent=-1.538, tangentMode=RCTM_User, time=1, value=0}
       ]}}})
   &Drag(Drag=3, Use Linear Drag=true, Use Rotational Drag=true)
   &CurlNoiseForce(
    Noise Frequency=20
    Noise Strength=&Multiply_Float(A=500, B=&Add_Float(A=$User.HitIntensity, B=0.3))
    Pan Noise=false
   )
   &SolveForcesAndVelocity(Clamp Velocity=false, Limit Acceleration=false)
   &ScaleColor(
    Scale Alpha=0.5
    Scale RGB=&MakeVectorFromLinearColorRGB(Color=&ColorFromCurve(ColorCurve=DataInterface() {propertyValues {
        AlphaCurve {keys [<at 0 1 RCIM_Cubic>]}
        BlueCurve {keys [<at 0 15 RCIM_Cubic>,<at 0.3966 1>,<at 1 1 RCIM_Cubic>]}
        GreenCurve {keys [<at 0 0 RCIM_Cubic>,<at 0.3966 0>,<at 1 0 RCIM_Cubic>]}
        RedCurve {keys [<at 0 14.6934 RCIM_Cubic>,<at 0.3966 0.4326>,<at 1 0.0897 RCIM_Cubic>]}
       }}))
    ScaleA=true
    ScaleRGB=true
   )
  ]
  rendererClasses [&NiagaraSpriteRendererProperties]
  renderers [{rendererClass=&NiagaraSpriteRendererProperties}]
 }
 Empty {
  emitterUpdateScript [
   &EmitterState()
   &SpawnRate(Spawn Probability=$User.Hit, SpawnRate=100, Use Spawn Probability=true) {enabled=false}
   &SpawnBurst_Instantaneous(Spawn Count=2, Use Spawn Probability=false, UseLoopCountLimit=false)
  ]
  particleSpawnScript [
   &InitializeParticle(
    Color=&MakeLinearColor(
     R:NiagaraFloat=1
     G:NiagaraFloat=1
     B:NiagaraFloat=1
     A:NiagaraFloat=&Add_Float(A=0.2, B=&Multiply_Float(A=0.8, B=$User.HitIntensity))
    )
    Lifetime=0.4
    Mesh Scale Mode=::NewEnumerator3
    Mesh Uniform Scale=&Multiply_Float(A=1.45, B=&Add_Float(A=1, B=&Multiply_Float(A=$User.HitIntensity, B=0.35)))
    Position Offset=<33 0 0>
    Sprite Size Mode=::NewEnumerator3
    Uniform Sprite Size=&Multiply_Float(A=200, B=&Add_Float(A=$User.HitIntensity, B=1))
    UsePositionOffset=true
    Write Material Random=true
   )
   &InitialMeshOrientation(Mesh Axis To Orient=<0 0 1>, Mesh Orientation Mode=::NewEnumerator2, Use Rotation Vector=false)
   &InitialMeshOrientation(
    Orientation Axis=<0 0 0>
    Orientation Vector=<0 0 0>
    Rotation=&RandomRangeVector(Minimum=<0 0 0>, Maximum=<1 0 0>)
    Use Orientation Vector=false
    Use Rotation Vector=true
   ) {module=InitialMeshOrientation001}
   /SetVariables(Particles.Intensity_Spawn:NiagaraFloat=&Add_Float(A=$User.HitIntensity, B=0)) {isSetParameters=true}
  ]
  particleUpdateScript [
   &ParticleState()
   &ScaleMeshSize(Scale Factor=&VectorFromFloat(Value=&FloatFromCurve(FloatCurve=DataInterface() {propertyValues {Curve {keys [{arriveTangent=0.5777, time=0, value=0.8}, {arriveTangent=0.4579, leaveTangent=0.4579, time=1, value=1}]}}})))
   &ScaleSpriteSize()
   &ScaleColor(
    Scale Alpha=&FloatFromCurve(FloatCurve=DataInterface() {propertyValues {Curve {keys [
         {arriveTangent=-3.1724, interpMode=RCIM_Cubic, leaveTangent=-3.1724, tangentMode=RCTM_User, time=0, value=1}
         <at 0.3 0.3>
         {interpMode=RCIM_Cubic, tangentMode=RCTM_User, time=1, value=0}
        ]}}})
    Scale RGB=&MakeVectorFromLinearColorRGB(Color=&ColorFromCurve(ColorCurve=DataInterface() {propertyValues {
        AlphaCurve {keys [<at 0 1 RCIM_Cubic>]}
        BlueCurve {keys [<at 0 3 RCIM_Cubic>,<at 0.5904 3 RCIM_Cubic>]}
        GreenCurve {keys [<at 0 0.8614 RCIM_Cubic>,<at 0.5904 0 RCIM_Cubic>]}
        RedCurve {keys [<at 0 2.6413 RCIM_Cubic>,<at 0.5904 0.1834 RCIM_Cubic>]}
       }}))
    ScaleA=true
    ScaleRGB=true
   )
   &FN_Radial_Material(
    Base Sprite Offset X:NiagaraFloat=&RandomRangeFloat(Minimum=0, Maximum=1)
    Base Sprite Offset Y:NiagaraFloat=&RandomRangeFloat(Minimum=0, Maximum=1)
    Base Sprite Tile U=6
    Base Sprite Tile V=2
    Radial Mask Offset X=0.3
    Radial Mask Power=5
    Radial Mask Strength=1.5
    Radial Mask Tile V=2
    Sprite Extrusion Speed=0.4
    Sprite Spin Speed=0.5
   ) {enabled=false}
   &FN_Sprite_Attributes(
    Color=&MultiplyLinearColor(
     Linear Color=&ColorFromCurve(ColorCurve=DataInterface() {propertyValues {
        AlphaCurve {keys [{arriveTangent=-0.3351, interpMode=RCIM_Cubic, leaveTangent=-0.3351, tangentMode=RCTM_User, time=0, value=0.2}, <at 1 0 RCIM_Cubic>]}
        BlueCurve {keys [<at 0 1 RCIM_Cubic>]}
        GreenCurve {keys [<at 0 0 RCIM_Cubic>]}
        RedCurve {keys [<at 0 0.1585 RCIM_Cubic>]}
       }})
     Scale Factor=&MakeVector4(X=1, Y=1, Z=1, W=$Particles.Intensity_Spawn)
    )
    Create Color Attribute=true
    Write to Module Namespace=true
   )
   &DynamicMaterialParameters(
    Index 0 Param 1=&FloatFromCurve(FloatCurve=DataInterface() {propertyValues {Curve {keys [{arriveTangent=-2.1015, interpMode=RCIM_Cubic, leaveTangent=-2.1015, tangentMode=RCTM_User, time=0, value=0.5}, <at 1 -0.15>]}}})
    Index 0 Param 2=&FloatFromCurve(FloatCurve=DataInterface() {propertyValues {Curve {keys [{arriveTangent=-0.3464, interpMode=RCIM_Cubic, leaveTangent=-0.3464, tangentMode=RCTM_User, time=0, value=1}, <at 1 0.9>]}}})
    Index 0 Param 3=&FloatFromCurve(FloatCurve=DataInterface() {propertyValues {Curve {keys [{interpMode=RCIM_Cubic, tangentMode=RCTM_User, time=0, value=1}, {interpMode=RCIM_Cubic, tangentMode=RCTM_User, time=1, value=0.25}]}}})
    Module.Param0WriteEnabled=true
    Module.Param1WriteEnabled=true
    Module.Param2WriteEnabled=true
    Module.Param3WriteEnabled=false
   )
   &FN_Light_Attributes(
    Color=<rgba 0.2579 0.1262 1 1>
    Create Color Attribute=true
    Create Exponent Attribute=true
    Create Radius Attribute=true
    Exponent=5
    Radius=150
    Scale Color=&FloatFromCurve(
     FloatCurve=DataInterface() {propertyValues {Curve {keys [
         {arriveTangent=-1.7915, interpMode=RCIM_Cubic, leaveTangent=-1.7915, tangentMode=RCTM_User, time=0, value=1}
         {interpMode=RCIM_Cubic, tangentMode=RCTM_User, time=1, value=0}
        ]}}}
     Scale Curve=&Add_Float(A=3.5, B=&Multiply_Float(A=6, B=$Particles.Intensity_Spawn))
    )
   )
  ]
  rendererClasses [
   &NiagaraSpriteRendererProperties
   &NiagaraMeshRendererProperties
   &NiagaraLightRendererProperties
  ]
  renderers [
   {rendererClass=&NiagaraSpriteRendererProperties}
   {rendererClass=&NiagaraMeshRendererProperties, rendererIndex=1}
   {rendererClass=&NiagaraLightRendererProperties, rendererIndex=2}
  ]
 }
}
summary {
 systemName=NS_Nebula_Emote_Drum
 emitters [
  {
   emitterName=Empty
   rendererClasses [
    &NiagaraSpriteRendererProperties
    &NiagaraMeshRendererProperties
    &NiagaraLightRendererProperties
   ]
  }
  {emitterName=Embers, rendererClasses [&NiagaraSpriteRendererProperties]}
 ]
 userVariables [
  User.Hit {type=NiagaraFloat, defaultValue=1, description=""}
  User.HitIntensity {type=NiagaraFloat, defaultValue=1, description=""}
 ]
}
