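# Niagara system for the Nebula drumsticks emote, drawing short-lived mesh particles on a skeletal mesh supplied through User.SkeletalMesh.
# One emitter spawns continuously, using two spawn groups as the bone or socket index to place oriented, camera-offset meshes.
# Dynamic material parameters, including a random value, drive the mesh material, and User.Transparency scales alpha.
# The burst spawn and inherited-movement modules are disabled.
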NiagaraSystem /BRCosmetics/Animation/Game/MainPlayer/Emotes/Nebula/NS_Nebula_Emote_Drumsticks
dependencies {usedDataInterfaces [], usedDynamicInputs [], usedModules [&SystemState]}
emitters {Trail_Spline_01 {
  emitterUpdateScript [
   &EmitterState(Life Cycle Mode=::NewEnumerator1, Loop Behavior=::NewEnumerator1, Loop Duration Mode=::NewEnumerator1, UseLoopDelay=false)
   &SpawnRate(SpawnRate=10, Use Spawn Probability=false)
   &SpawnBurst_Instantaneous(Spawn Count=2, Use Spawn Probability=false, UseLoopCountLimit=false) {enabled=false}
  ]
  particleSpawnScript [
   &FN_SetSpawnGroup(Number of Spawn Groups=2)
   &InitializeParticle(
    Color=<rgba 1 1 1 0.6>
    Lifetime=0.2
    Mesh Scale=<10 10 10>
    Mesh Scale Mode=::NewEnumerator4
    UsePositionOffset=false
    Write Material Random=true
   )
   &InitialMeshOrientation(Mesh Axis To Orient=<0 0 -1>, Mesh Orientation Mode=::NewEnumerator2, Orientation Vector=<0 0 1>, Use Rotation Vector=true)
   &CameraOffset(Camera Offset Amount=1.5)
   &SkeletalMeshLocation(
    Bone / Socket Index=$Particles.SpawnGroup
    Bone Sampling Mode=::NewEnumerator4
    MaskSpawnGroup=false
    Sampled Position Offset=<3 0 0>
    Skeletal Mesh=$User.SkeletalMesh
   )
  ]
  particleUpdateScript [
   &ParticleState()
   &DynamicMaterialParameters(
    Index 0 Param 1=&RandomRangeFloat(Minimum=0, Maximum=1)
    Index 0 Param 2=22
    Index 0 Param 3=1
    Index 0 Param 4=1
    Index 1 Param 2=0.25
    Index 3 Param 2=2
    Module.Param0WriteEnabled=true
    Module.Param0WriteEnabled1=true
    Module.Param0WriteEnabled3=true
    Module.Param1WriteEnabled=true
    Module.Param1WriteEnabled1=true
    Module.Param1WriteEnabled3=true
    Module.Param2WriteEnabled=true
    Module.Param2WriteEnabled1=true
    Module.Param2WriteEnabled3=true
    Module.Param3WriteEnabled=true
    Module.Param3WriteEnabled1=true
    Module.Param3WriteEnabled3=true
    Write Parameter Index 1=true
    Write Parameter Index 3=true
   )
   &ScaleColor(Scale Alpha=$User.Transparency, ScaleA=true, ScaleRGB=false)
   &InheritSourceMovement(
    Applied Position Scale=<0.5 0.5 0.5>
    ApplyForce=false
    ApplyPosition=true
    ApplyPositionFalloff=false
    Source Velocity:Vector3f=$Particles.SkeletalMeshLocation.SampledVelocity
   ) {enabled=false}
  ]
  rendererClasses [&NiagaraMeshRendererProperties]
  renderers [{rendererClass=&NiagaraMeshRendererProperties, rendererIndex=0}]
 }}
summary {
 systemName=NS_Nebula_Emote_Drumsticks
 emitters [{emitterName=Trail_Spline_01, rendererClasses [&NiagaraMeshRendererProperties]}]
 userVariables [
  User.bIsFrontEnd {type=NiagaraBool}
  User.bIsFrontEndPreview {type=NiagaraBool}
  User.ColorProgress {type=NiagaraFloat, defaultValue=0, description=""}
  User.DissolveProgress {type=NiagaraFloat, defaultValue=0, description=""}
  User.Interpolate Intensity {type=NiagaraFloat, defaultValue=0, description=""}
  User.RibbonIDOffset {type=NiagaraInt32}
  User.ScaleMeshSize {type=Vector3f, description="", defaultValue {struct=Vector3f, value=<0 0 0>}}
  User.Skeletal Mesh {
   type=NiagaraDataInterfaceSkeletalMesh
   defaultValue {
    struct=NiagaraExt_VariableValue_DataInterface
    value {dataInterface=&self:NiagaraDataInterfaceSkeletalMesh_2}
   }
  }
  User.SkeletalMesh {
   type=NiagaraDataInterfaceSkeletalMesh
   defaultValue {
    struct=NiagaraExt_VariableValue_DataInterface
    value {dataInterface=&self:NiagaraDataInterfaceSkeletalMesh_0}
   }
  }
  User.Transparency {type=NiagaraFloat, defaultValue=1, description=""}
 ]
}
